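# Sound cue that plays a looping ambient sound for a rune.
# A looping wave player feeds the Rune_Ambient_Loop_01 wave into an Enveloper root, whose volume and pitch curves each hold a single key, so neither varies over time.
# The cue uses a close-range prop attenuation and is assigned to the Fort_Impacts sound class.
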SoundCue /PrimalGameplay/Sounds/Rune/Rune_Ambient_Loop_Cue
root=Enveloper_0
nodes [
 {
  id=Enveloper_0
  children [WavePlayer_0]
  properties {pitchCurve {editorCurveData {keys [{time=0}]}}, volumeCurve {editorCurveData {keys [<at 0 1>]}}}
 }
 {id=WavePlayer_0, properties {bLooping=true, soundWaveAssetPtr=&/Rune_Ambient_Loop_01}}
]
properties {
 attenuationSettings=/Game/Sounds/Attenuation/Fort_Prop_Close_Att
 maxDistance=1800
 soundClassObject=&Fort_Impacts
}
