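# Sound cue for rune physics impacts, played as a close-range attenuated impact sound.
# An ImpactType switch selects random light or medium waves: its first two inputs share the light branch, and the last two branches both use medium-named waves.
# In each branch, a Strength parameter raises volume and lowers pitch as it increases.
# A root modulator adds slight random pitch and volume variation.
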SoundCue /PrimalGameplay/Sounds/Rune/Rune_Impact_Phys_Cue
root=Modulator_1
nodes [
 {
  id=ModulatorContinuous_0
  children [Random_0]
  properties {
   pitchModulationParams {default=0, maxOutput=0.8, minOutput=1.2, parameterName=Strength}
   volumeModulationParams {default=0, minOutput=0.2, parameterName=Strength}
  }
 }
 {
  id=ModulatorContinuous_1
  children [Random_3]
  properties {
   pitchModulationParams {default=0, maxOutput=0.8, minOutput=1.2, parameterName=Strength}
   volumeModulationParams {default=0, maxOutput=1.2, minOutput=0.8, parameterName=Strength}
  }
 }
 {
  id=ModulatorContinuous_2
  children [Random_4]
  properties {
   pitchModulationParams {default=0, maxOutput=0.8, minOutput=1.2, parameterName=Strength}
   volumeModulationParams {default=0, maxOutput=1.2, minOutput=0.8, parameterName=Strength}
  }
 }
 {id=Modulator_1, children [Switch_0], properties {pitchMax=1.1, pitchMin=0.9, volumeMax=1.1, volumeMin=0.9}}
 {id=Random_0, children [WavePlayer_0,WavePlayer_1,WavePlayer_2], properties {weights [1,1,1]}}
 {id=Random_3, children [WavePlayer_24, WavePlayer_25, WavePlayer_26], properties {weights [1,1,1]}}
 {id=Random_4, children [WavePlayer_27, WavePlayer_28, WavePlayer_29], properties {weights [1,1,1]}}
 {
  id=Switch_0
  children [ModulatorContinuous_0, ModulatorContinuous_0, ModulatorContinuous_1, ModulatorContinuous_2]
  properties {intParameterName=ImpactType}
 }
 {id=WavePlayer_0, properties {soundWaveAssetPtr=&/Rune_Impact_Light_01}}
 {id=WavePlayer_1, properties {soundWaveAssetPtr=&/Rune_Impact_Light_02}}
 {id=WavePlayer_2, properties {soundWaveAssetPtr=&/Rune_Impact_Light_03}}
 {id=WavePlayer_24, properties {soundWaveAssetPtr=&/Rune_Impact_Medium_01}}
 {id=WavePlayer_25, properties {soundWaveAssetPtr=&/Rune_Impact_Medium_02}}
 {id=WavePlayer_26, properties {soundWaveAssetPtr=&/Rune_Impact_Medium_03}}
 {id=WavePlayer_27, properties {soundWaveAssetPtr=&Rune_Impact_Medium_01}}
 {id=WavePlayer_28, properties {soundWaveAssetPtr=&Rune_Impact_Medium_02}}
 {id=WavePlayer_29, properties {soundWaveAssetPtr=&Rune_Impact_Medium_03}}
]
properties {
 attenuationSettings=/Game/Sounds/Attenuation/Fort_Prop_Close_Att
 maxDistance=1800
 soundClassObject=&Fort_Impacts
 volumeMultiplier=1.25
}
