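# Sound cue for picking up a rune, layering a rune-specific sound over a generic large held-object pickup.
# A random node picks one of three held-object pickup variations, mixed at equal volume with the rune wave, which starts after a short fixed delay.
# A modulator raises the pitch of the whole mix by a slightly random amount.
# It uses close-prop attenuation and the impacts sound class.
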SoundCue /PrimalGameplay/Sounds/Rune/Rune_Pickup_Cue
root=Modulator_2
nodes [
 {id=Delay_0, children [WavePlayer_0], properties {delayMax=0.1, delayMin=0.1}}
 {id=Mixer_0, children [Delay_0,Random_0], properties {inputVolume [1,1]}}
 {id=Modulator_2, children [Mixer_0], properties {pitchMax=1.2, pitchMin=1.1}}
 {id=Random_0, children [WavePlayer_1, WavePlayer_2, WavePlayer_10], properties {weights [1,1,1]}}
 {id=WavePlayer_0, properties {soundWaveAssetPtr=&/Rune_Pickup_01}}
 {id=WavePlayer_1, properties {soundWaveAssetPtr=/FNGameplayCues/Sounds/Apollo/HeldObjects/HeldObjects_Large_Pickup_01}}
 {id=WavePlayer_10, properties {soundWaveAssetPtr=/FNGameplayCues/Sounds/Apollo/HeldObjects/HeldObjects_Large_Pickup_03}}
 {id=WavePlayer_2, properties {soundWaveAssetPtr=/FNGameplayCues/Sounds/Apollo/HeldObjects/HeldObjects_Large_Pickup_02}}
]
properties {
 attenuationSettings=/Game/Sounds/Attenuation/Fort_Prop_Close_Att
 maxDistance=1800
 soundClassObject=&Fort_Impacts
 volumeMultiplier=1
}
