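# Looping sound cue for a rolling Rune physics prop, with pitch and volume that respond to a Speed parameter.
# A looping wave player feeds a continuous modulator, which maps Speed to a bounded pitch range and also to volume, both defaulting to zero input.
# The cue uses a dedicated rune roll attenuation asset and the Fort_Impacts sound class, with boosted overall pitch and volume.
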SoundCue /PrimalGameplay/Sounds/Rune/Rune_Roll_Phys_Loop_Cue
root=ModulatorContinuous_0
nodes [
 {
  id=ModulatorContinuous_0
  children [WavePlayer_1]
  properties {
   pitchModulationParams {default=0, maxOutput=1.4, minOutput=0.8, parameterName=Speed}
   volumeModulationParams {default=0, parameterName=Speed}
  }
 }
 {id=WavePlayer_1, properties {bLooping=true, soundWaveAssetPtr=&/Rune_Rolling_Loop_01}}
]
properties {
 attenuationSettings=&/Fort_Prop_Rune_Roll_Att
 maxDistance=1000
 pitchMultiplier=1.25
 soundClassObject=&Fort_Impacts
 volumeMultiplier=2.25
}
