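# Sound cue for throwing a rune, routed through the ranged weapon sound class with close-prop attenuation at reduced volume.
# It picks one of two throw waves at equal weight, varies it through a modulator that raises pitch slightly, and starts it after a short fixed delay.
# A concurrency override prevents new instances from starting once the voice limit is reached, rather than cutting off playing ones.
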SoundCue /PrimalGameplay/Sounds/Rune/Rune_Throw_Cue
root=Delay_0
nodes [
 {id=Delay_0, children [Modulator_0], properties {delayMax=0.04, delayMin=0.04}}
 {id=Modulator_0, children [Random_1], properties {pitchMax=1.2, pitchMin=1.1}}
 {id=Random_1, children [WavePlayer_3,WavePlayer_4], properties {weights [1,1]}}
 {id=WavePlayer_3, properties {soundWaveAssetPtr=&/Rune_Throw_01}}
 {id=WavePlayer_4, properties {soundWaveAssetPtr=&/Rune_Throw_02}}
]
properties {
 attenuationSettings=/Game/Sounds/Attenuation/Fort_Prop_Close_Att
 bOverrideConcurrency=true
 maxDistance=1800
 soundClassObject=&Fort_WeaponRanged
 volumeMultiplier=0.35
 concurrencyOverrides {resolutionRule=PreventNew}
}
